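; Supply drop balloon actor for Athena that players can shoot down, with a global damage-reduction gameplay effect applied at BeginPlay.
; Server damage shows its normally hidden health bar, restarting a timer that hides it again.
; Server death notifies the supply drop, starts a short lifespan and logs a shot-down analytics event, while death effects hide the balloon and play pop FX and sound.
; Marking resolves to the parent crate; movement-stopped handling is a stub for children.
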(event EventBeginPlay
  ;[1 1 1] hide health bar on beginplay
  (HealthBarComponent|SetHealthBarVisibility (Variables|Default|GetFortHealthBar) false)
  (Environment|ApplyGlobalEnvironmentGameplayEffecttoActor self (Variables|Default|GetGE_DamageReduction)))

(event EventOnDamageServer
  (params (Damage :type Float%20%28single-precision%29) (DamageTags :type Gameplay%20Tag%20Container%20Structure%20%28by%20ref%29) (Momentum :type Vector :default "0, 0, 0") (HitInfo :type Hit%20Result%20Structure%20%28by%20ref%29) (InstigatedBy :type Controller%20Object%20Reference) (DamageCauser :type Actor%20Object%20Reference) (EffectContext :type Gameplay%20Effect%20Context%20Handle%20Structure))
  ;[1 1 1] handle damage
  (if (Variables|Actor|GetbCanBeDamaged)
    ;;[1 1 1] make hp bar visible after taking damage
    (HealthBarComponent|SetHealthBarVisibility (Variables|Default|GetFortHealthBar) true)
    (Utilities|Time|ClearandInvalidateTimerbyHandle (Variables|Default|GetHealthbarVisibleTimerHandle))
    (bind _returnvalue (Utilities|Time|SetTimerbyEvent (delegate Custom|OnHealthbarVisibleTimerExpired) (Variables|Default|GetHealthbarVisibleTime)))
    (Variables|Default|SetHealthbarVisibleTimerHandle _returnvalue)))

(event Custom|OnHealthbarVisibleTimerExpired
  ;[1 1 1] handle damage
  ;;[1 1 1] make hp bar visible after taking damage
  (HealthBarComponent|SetHealthBarVisibility (Variables|Default|GetFortHealthBar) false))

(event EventOnDeathServer
  (params (Damage :type Float%20%28single-precision%29) (DamageTags :type Gameplay%20Tag%20Container%20Structure%20%28by%20ref%29) (Momentum :type Vector :default "0, 0, 0") (HitInfo :type Hit%20Result%20Structure%20%28by%20ref%29) (InstigatedBy :type Controller%20Object%20Reference) (DamageCauser :type Actor%20Object%20Reference) (EffectContext :type Gameplay%20Effect%20Context%20Handle%20Structure))
  ;[1 1 1] OnDeathServer
  (Utilities|FlowControl|Sequence
    (:then_0
      (Variables|Actor|SetbCanBeDamaged false)
      (Default|CallOnBalloonDestroyed) ; Let supply drop know
      (Actor|SetLifeSpan 2.0)
      (Utilities|Time|ClearandInvalidateTimerbyHandle (Variables|Default|GetHealthbarVisibleTimerHandle)))
    (:then_1
      ;;[1 1 1] Fire analytics event indicating that the balloon was shot down
      (Analytics|FireAnalyticsEventBlueprint "AthenaSupplyDrop.ShotDown" (Utilities|Struct|MakeFortAnalyticsEventAttribute "(Name=\"\",Value=\"\")")))))

(event EventOnDeathPlayEffects
  (params (Damage :type Float%20%28single-precision%29) (DamageTags :type Gameplay%20Tag%20Container%20Structure%20%28by%20ref%29) (Momentum :type Vector :default "0, 0, 0") (HitInfo :type Hit%20Result%20Structure%20%28by%20ref%29) (InstigatedBy :type Fort%20Pawn%20Object%20Reference) (DamageCauser :type Actor%20Object%20Reference) (EffectContext :type Gameplay%20Effect%20Context%20Handle%20Structure))
  (bind _staticmeshcomponent (Variables|Debug|GetStaticMeshComponent))
  (bind _returnvalue (Transformation|GetSocketLocation _staticmeshcomponent "boom"))
  ;[1 1 1] OnDeathPlayeffects
  (Rendering|SetActorHiddenInGame true)
  (bind _returnvalue_2 (Effects|Components|ParticleSystem|SpawnFXAtLocation (Variables|Default|GetBalloonPopFX) _returnvalue))
  (Components|Activation|SetActive _returnvalue_2 true false)
  (Audio|PlaySoundatLocation (Variables|Default|GetBalloonPopSound) _returnvalue)
  (HealthBarComponent|SetHealthBarVisibility (Variables|Default|GetFortHealthBar) false))

(fn GetAttachedActorToMark
  ;[0.01601 0.1969 0.2344] For default supply drops, the parent actor itself is the crate that is falling.
  (return (Actor|GetParentActor)))

(fn HandleSupplyDropMovementStopped
  (params (ImpactResult :type Hit%20Result%20Structure%20%28by%20ref%29))
  ;[0.0211 0.1616 0.1719] Override in children.
  (return))

(fn ConstructionScript)
