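; Battle Royale (Athena) supply drop crate that falls under a damageable balloon, lands, and spawns tiered loot when searched.
; A curve-table-tuned timeline lerps the crate to ground; popping the balloon raises fall speed.
; Collision intercepts land it on buildings, while server-bound death events on supporting actors recompute landing and resume falling.
; Multicasts and RepNotifies sync landing, re-falls and loot FX, with special handling for replays.
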(event EventBeginPlay
  ;[1 1 1] begin play
  (Variables|Default|SetEventData (Utilities|Struct|MakeGameplayEventData :Instigator self :Target self :ContextHandle "()" :InstigatorTags "(GameplayTags=)" :TargetTags "(GameplayTags=)" :EventMagnitude 0.0 :TargetData "()"))
  (Utilities|FlowControl|Sequence
    (:then_0
      (CallFunction|RegisterforBalloonDestruction)
      (CallFunction|InitFall))
    (:then_1
      ;;[1 1 1] Supply drop spawned analytics event
      (Networking|SwitchHasAuthority
        (:Authority
          (Analytics|FireAnalyticsEventBlueprint "AthenaSupplyDrop.Spawned" (Utilities|Struct|MakeFortAnalyticsEventAttribute "(Name=\"\",Value=\"\")")))))))

(event Custom|OnLandingLocationChanged
  (params (DamagedActor :type Actor%20Object%20Reference) (Damage :type Float%20%28single-precision%29) (InstigatedBy :type Controller%20Object%20Reference) (DamageCauser :type Actor%20Object%20Reference) (HitLocation :type Vector) (FHitComponent :type Primitive%20Component%20Object%20Reference) (BoneName :type Name) (Momentum :type Vector))
  ;[1 1 1] Event fired when the ground this supply drop was set to land on has died
  ;;[1 0.9783 0.2889] Falling behaves differently if we've already landed and cleared the reticle effect - we'll lerp to the ground instead of directly setting actor location
  (Utilities|IsValid (Variables|Default|GetFloor)
    (:Is%20Not%20Valid
      (bind _returnvalue (GroundCollision|FindGroundLocationAt (Transformation|GetActorLocation)))
      (Transformation|SetActorLocation _returnvalue false true)
      (CallFunction|InitFall))))

(event OnComponentHit(InterceptCollision)
  (params (HitComponent :type Primitive%20Component%20Object%20Reference) (OtherActor :type Actor%20Object%20Reference) (OtherComp :type Primitive%20Component%20Object%20Reference) (NormalImpulse :type Vector :default "0, 0, 0") (Hit :type Hit%20Result%20Structure%20%28by%20ref%29))
  ;[1 1 1] handle landing on destructable buildings and props
  (if (and (Utilities|NotEqual(Object) OtherActor self) (not (Utilities|IsValid (Variables|Default|GetFloor) :!class /Script/Engine.KismetSystemLibrary)))
    (bind _asbuilding_actor (Utilities|Casting|CastToBuildingActor OtherActor))
    (Variables|Default|SetFloor _asbuilding_actor)
    (Utilities|FlowControl|Sequence
      (:then_0
        (Transformation|SetActorLocation (- (Transformation|GetWorldLocation (Variables|Default|GetSupplyDrop_Mesh)) "0, 0,10.000000"))
        (CallFunction|DropIntercepted))
      (:then_1
        ;[1 1 1] Bind the current "Floor" to a OnDied event so we know to recalculate the crate position
        (Networking|SwitchHasAuthority
          (:Authority
            (EventDispatchers|BindEventtoOnDied (Variables|Default|GetFloor) (delegate Custom|OverlappingDied))))))))

(event Custom|RegisterForBalloonDestruction
  ;[1 1 1] Balloon is destroyed - turn on falling trail and lerp falling playrate to 2
  (bind _asbp_damage_balloon_athena (Utilities|Casting|CastToBP_DamageBalloon_Athena (Variables|/Script/Engine.ChildActorComponent|GetChildActor (Variables|Default|GetBP_DamageBalloon_Athena))))
  (Default|BindEventtoOnBalloonDestroyed _asbp_damage_balloon_athena (delegate Custom|BalloonDestroyed)))

(event OnComponentBeginOverlap(InterceptCollision)
  (params (OverlappedComponent :type Primitive%20Component%20Object%20Reference) (OtherActor :type Actor%20Object%20Reference) (OtherComp :type Primitive%20Component%20Object%20Reference) (OtherBodyIndex :type Integer) (bFromSweep :type Boolean) (SweepResult :type Hit%20Result%20Structure%20%28by%20ref%29))
  ;[1 1 1] handle landing on destructable buildings and props
  (if (and (Utilities|NotEqual(Object) OtherActor self) (not (Utilities|IsValid (Variables|Default|GetFloor) :!class /Script/Engine.KismetSystemLibrary)))
    (bind _asbuilding_actor (Utilities|Casting|CastToBuildingActor OtherActor))
    (Variables|Default|SetFloor _asbuilding_actor)
    (Utilities|FlowControl|Sequence
      (:then_0
        (Transformation|SetActorLocation (- (Transformation|GetWorldLocation (Variables|Default|GetSupplyDrop_Mesh)) "0, 0,10.000000"))
        (CallFunction|DropIntercepted))
      (:then_1
        ;[1 1 1] Bind the current "Floor" to a OnDied event so we know to recalculate the crate position
        (Networking|SwitchHasAuthority
          (:Authority
            (EventDispatchers|BindEventtoOnDied (Variables|Default|GetFloor) (delegate Custom|OverlappingDied))))))))

(event Custom|ImpactFX
  ;[1 1 1] Impact
  (Components|Activation|Deactivate (Variables|Default|GetNS_SupplyDrop_Active))
  (Audio|Components|Audio|Stop (Variables|Default|GetReticleAudio))
  ;[1 1 1] Post and Shake
  (Audio|PlaySoundatLocation (Variables|Audio|GetImpact_Sound) (Transformation|GetActorLocation))
  (Effects|SpawnEmitteratLocation "/Game/Gadgets/Assets/SupplyDrop/Effects/P_SupplyDrop_Impact.P_SupplyDrop_Impact" (CallFunction|GetImpactLocation) "90.000000,90.000000,0.000000") ; Impact Effect
  (Networking|CanExecuteCosmeticEvents
    (:True
      (bind _returnvalue (Game|SpawnActorfromClass "/Game/Effects/Fort_Effects/BluePrints/PostProcessFX/B_Post_SupplyDrop_Impact.B_Post_SupplyDrop_Impact_C" (Transformation|GetActorTransform) "AlwaysSpawn" "OverrideRootScale" 0 "400.000000,400.000000,400.000000" 0))
      (Class|BPostSupplyDropImpact|SpawnPostFX _returnvalue 1.3e+03 "0, 0, 0" true 2 true))))

(event Custom|BalloonDestroyed
  ;[1 1 1] Balloon is destroyed - turn on falling trail and lerp falling playrate to 2
  (Networking|SwitchHasAuthority
    (:Authority
      (CallFunction|MulticastBalloonDestroyed))))

(event Custom|Multicast%20BalloonDestroyed
  ;[1 1 1] Multicast  Balloon Destroyed
  (Audio|SpawnSoundAttached "/Game/Sounds/Foley_Loot/Containers/SupplyDrops/Fort_SupplyDrop_Balloon_Explode_Cue.Fort_SupplyDrop_Balloon_Explode_Cue" (Variables|Default|GetSupplyDrop_Mesh))
  (Variables|Default|SetbBalloonPopped true)
  (CallFunction|ClientPlayDropFX)
  (if (not (Variables|Default|GetAllow_SupplyDrop_Interact))
    (CallFunction|ToggleFallingFX true)
    (bind (_direction _lerp) (Timeline|LerpFallRate PlayFromStart)
      (:Update
        (Components|Timeline|SetPlayRate (Variables|AthenaSupplyDrop_02|GetFallingTimeline) (Math|Float|Lerp (Variables|Default|GetFallingPlayrate) 2 _lerp))))))

(event Custom|OverlappingDied
  (params (DamagedActor :type Actor%20Object%20Reference) (Damage :type Float%20%28single-precision%29) (InstigatedBy :type Controller%20Object%20Reference) (DamageCauser :type Actor%20Object%20Reference) (HitLocation :type Vector) (FHitComponent :type Primitive%20Component%20Object%20Reference) (BoneName :type Name) (Momentum :type Vector))
  (bind _returnvalue (Transformation|GetActorLocation))
  (bind _z (.z _returnvalue))
  (bind _returnvalue_2 (- _z 50000))
  (bind _y (.y _returnvalue))
  (bind _x (.x _returnvalue))
  (bind _returnvalue_3 (Math|Vector|MakeVector _x _y _returnvalue_2))
  (bind (_outhit _returnvalue_4) (GroundCollision|GroundTrace _returnvalue _returnvalue_3))
  (bind (_bblockinghit _binitialoverlap _time _distance _location _impactpoint) (Collision|BreakHitResult _outhit))
  ;[1 1 1] Bind the current "Floor" to a OnDied event so we know to recalculate the crate position
  (Variables|Default|SetFloor 0)
  (if _returnvalue_4
    (Variables|Default|SetNewLocation _impactpoint)))

(event OnComponentBeginOverlap(TargetCollision)
  (params (OverlappedComponent :type Primitive%20Component%20Object%20Reference) (OtherActor :type Actor%20Object%20Reference) (OtherComp :type Primitive%20Component%20Object%20Reference) (OtherBodyIndex :type Integer) (bFromSweep :type Boolean) (SweepResult :type Hit%20Result%20Structure%20%28by%20ref%29))
  ;[1 1 1] target location overlap (used for making sure we can adjust destination if base is destroyed before drop has landed)
  (if (Utilities|NotEqual(Object) OtherActor self)
    (bind _asbuilding_actor (Utilities|Casting|CastToBuildingActor OtherActor))
    (Variables|Default|SetFloor _asbuilding_actor)
    ;[1 1 1] Bind the current "Floor" to a OnDied event so we know to recalculate the crate position
    (Networking|SwitchHasAuthority
      (:Authority
        (EventDispatchers|BindEventtoOnDied (Variables|Default|GetFloor) (delegate Custom|OverlappingDied))))))

(event Custom|PlayDrop
  (bind _interceptcollision (Variables|Default|GetInterceptCollision))
  ;[1 1 1] Initial fall
  (Utilities|FlowControl|Sequence
    (:then_0
      (Collision|IgnoreActorwhenMoving _interceptcollision (Variables|/Script/Engine.ChildActorComponent|GetChildActor (Variables|Default|GetBP_DamageBalloon_Athena)) true)
      (Collision|SetCollisionEnabled _interceptcollision "QueryOnly")
      (Transformation|SetWorldLocation (Variables|Default|GetInterceptCollision) (CallFunction|GetFallLocation) false false) ; Set initial location without doing a sweep, to fix issues with spawning over water
    )
    (:then_1
      (bind _shouldplay (CallFunction|HandleFallTimeline))
      (if _shouldplay
        (bind (_direction _fallcurve _fallcrossfadeaudiocurve) (Timeline|FallingTimeline Play :NewTime 0.999)
          (:Update
            (CallFunction|UpdateFall _fallcurve _fallcrossfadeaudiocurve))
          (:Finished
            (CallFunction|FallTimelineFinished))
          (:DisableFlare
            (CallFunction|DisableFlare))
          (:Almost%20Landed
            (CallFunction|AlmostLanded)))))
    (:then_2
      (Utilities|FlowControl|Delay 0.2)
      (CallFunction|StartDropAudio))))

(event Custom|DropIntercepted
  ;[1 1 1] Initial fall
  (Timeline|FallingTimeline SetNewTime))

(event Custom|PlayDropNewTime
  ;[1 1 1] Update position if parent floor is destroyed
  (bind _returnvalue (Utilities|Time|SetTimerbyEvent (delegate Custom|SecondaryFallTick) (Utilities|Time|GetWorldDeltaSeconds) true))
  (Variables|Default|SetSecondaryFallingTimerHandle _returnvalue)
  (Variables|Default|SetFloor 0)
  (Transformation|SetRelativeLocation (Variables|Default|GetInterceptCollision) "0, 0,2.000000" false false))

(event Custom|SpawnLootDestroyActor
  ;[1 1 1] SpawnLoot - from timer or by looting
  (CallFunction|SpawnLoot)
  (Gameplay|SilentDie false))

(event Custom|SecondaryFallTick
  ;[1 1 1] Update position if parent floor is destroyed
  (Transformation|SetActorLocation (Math|Vector|MakeVector (.x (Transformation|GetActorLocation)) (.y (Transformation|GetActorLocation)) (- (.z (Transformation|GetActorLocation)) 16)) false true)
  (if (< (.z (Transformation|GetActorLocation)) (.z (Variables|Default|GetNewLocation)))
    (Transformation|SetActorLocation (Variables|Default|GetNewLocation) false true)
    (Utilities|Time|ClearandInvalidateTimerbyHandle (Variables|Default|GetSecondaryFallingTimerHandle))))

(event Interaction|EventBlueprintOnInteract
  (params (InteractingPawn :type Fort%20Pawn%20Object%20Reference) (InteractionBeingAttempted :type EInteractionBeingAttempted%20Enum :default "FirstInteraction"))
  ;[1 1 1] Interact
  (Utilities|FlowControl|Sequence
    (:then_0
      (Variables|Default|SetMyInteractingPawn InteractingPawn)
      (CallFunction|SpawnLootDestroyActor)
      (CallFunction|StopBeep)
      ;[1 1 1] let balloon know the crate has been looted
      (bind _asbp_damage_balloon_athena (Utilities|Casting|CastToBP_DamageBalloon_Athena (Variables|/Script/Engine.ChildActorComponent|GetChildActor (Variables|Default|GetBP_DamageBalloon_Athena))))
      (Class|BPDamageBalloonAthena|CrateIsGone _asbp_damage_balloon_athena))
    (:then_1
      (Variables|Default|SetMyInteractingPawn 0))))

(event Custom|StopBeep
  ;[1 1 1] Stop beep
  (Audio|PlaySoundatLocation (Variables|Audio|GetLoot_Sound) (Transformation|GetActorLocation)))

(event OnComponentActivated(BP_DamageBalloon_Athena)
  (params (Component :type Actor%20Component%20Object%20Reference) (bReset :type Boolean)))

(event Custom|Multicast%20Finish%20Fall
  (params (FinalLocation :type Vector))
  ;[1 1 1] Multicast Finish Fall - moved network essentials here for synchronization
  (Components|Timeline|Stop (Variables|AthenaSupplyDrop_02|GetFallingTimeline))
  (bind _result (CallFunction|PlayingReplay))
  (if (not _result)
    (Variables|Default|SetbLanded true))
  (Transformation|SetWorldLocation (Variables|Default|GetInterceptCollision) FinalLocation false true)
  (CallFunction|TurnonInteract true)
  (Components|Activation|Deactivate (Variables|Default|GetP_Flare))
  (CallFunction|ToggleFallingFX false)
  (CallFunction|ImpactFX)
  (Ability|SendGameplayEventtoActor self (Variables|Default|GetEventHit) (Variables|Default|GetEventData)))

(fn AlmostLanded
  (if (not (Variables|Default|GetbBalloonPopped))
    (Audio|SpawnSoundAttached "/Game/Sounds/Foley_Loot/Containers/SupplyDrops/Fort_SupplyDrop_LastSecond_Athena_Cue.Fort_SupplyDrop_LastSecond_Athena_Cue" (Variables|Default|GetSupplyDrop_Mesh))))

(fn BlueprintCanInteract
  (params (InteractingPawn :type Fort%20Pawn%20Object%20Reference) (InteractionBeingAttempted :type EInteractionBeingAttempted%20Enum :default "FirstInteraction") (InteractionType :type TInteractionType%20Enum :default "IT_NoInteraction"))
  (return (Variables|Default|GetAllow_SupplyDrop_Interact)))

(fn BlueprintGetInteractionString
  (params (InteractingPawn :type Fort%20Pawn%20Object%20Reference) (InteractionBeingAttempted :type EInteractionBeingAttempted%20Enum :default "FirstInteraction"))
  (return "NSLOCTEXT(\"[08CF574946270D8A172FB29F8AD5A096]\", \"9FF5FC4F4B8098BB5C4577B86C59D119\", \"Search\")"))

(fn ClientPlayDropFX
  (Networking|CanExecuteCosmeticEvents
    (:True
      (bind _activate_sound (Variables|Audio|GetActivate_Sound)
        (:then
          (Audio|PlaySoundatLocation _activate_sound (Transformation|GetActorLocation)))
        (:else)))))

(fn DisableFlare
  (Components|Activation|Deactivate (Variables|Default|GetP_Flare))
  (Audio|Components|Audio|FadeOut (Variables|Default|GetFlareAudio) 3 0.0 "Linear" :!class /Script/Engine.AudioComponent))

(fn FallTimelineFinished
  (Components|Timeline|Stop (Variables|AthenaSupplyDrop_02|GetFallingTimeline))
  (Networking|SwitchHasAuthority
    (:Authority
      (CallFunction|MulticastFinishFall (CallFunction|GetImpactLocation))
      (CallFunction|MoveOverlappingPawns))))

(fn GetFallLocation
  (return (+ (CallFunction|GetImpactLocation) (Math|Vector|MakeVector 0.0 0.0 (Variables|Default|GetFallHeight)))))

(fn GetImpactLocation
  (return (Transformation|GetActorLocation)))

(fn HandleFallTime
  (bind _outxy (Fortnite|EvaluateCurveTableRow (Variables|Default|GetCurveTable_FallTime)))
  (bind _output_get (Variables|Default|SetFallingPlayrate (/ 1 _outxy)))
  (Components|Timeline|SetPlayRate (Variables|AthenaSupplyDrop_02|GetFallingTimeline) _output_get))

(fn HandleFallTimeline
  (bind _falltime (Variables|Default|GetFallTime))
  (bind _returnvalue (- 1.0 _falltime))
  (bind _fallingtimeline (Variables|AthenaSupplyDrop_02|GetFallingTimeline))
  (bind _blanded (Variables|Default|GetbLanded))
  (bind _result (CallFunction|PlayingReplay))
  (if _result
    (Components|Timeline|SetNewTime _fallingtimeline _returnvalue)
    (if _blanded
      ;[1 1 1 1 11] These might not be necessary, but since the timeline won't be playing, I'm going to do them here anyways
      (CallFunction|DisableFlare)
      (CallFunction|AlmostLanded)
      (else
        (CallFunction|UpdateFall _falltime _returnvalue true)))
    (return (not _blanded))
    (else
      (Components|Timeline|PlayfromStart _fallingtimeline)
      (return true))))

(fn HideAndExplodeCrate
  (bind _p_supplydropcrate_explosion (Variables|Default|GetP_SupplydropCrate_Explosion))
  (Development|SetHiddeninGame (Variables|Default|GetSupplyDrop_Mesh) true false)
  (Components|Activation|Activate _p_supplydropcrate_explosion true)
  (Transformation|DetachFromComponent _p_supplydropcrate_explosion "KeepWorld" "KeepWorld" "KeepWorld" true))

(fn InitDrop
  (params (InUnlocks :type Supply%20Drop%20Unlocks%20Structure))
  ;[1 1 1] Called by Ability
  (Variables|Default|SetUnlocks InUnlocks))

(fn InitFall
  (bind (_bblockinghit _binitialoverlap _time _distance _location _impactpoint _normal _impactnormal _physmat _hitactor) (Collision|BreakHitResult (Variables|FortAthenaSupplyDrop|GetLastGroundLocationHitResult)))
  (Utilities|FlowControl|Sequence
    (:then_0
      ;[1 1 1] Bind to handle ground below us destroyed as needed
      (Networking|SwitchHasAuthority
        (:Authority
          (bind _asbuilding_actor (Utilities|Casting|CastToBuildingActor _hitactor))
          (EventDispatchers|BindEventtoOnDied _asbuilding_actor (EventDispatchers|CreateEvent :!event OnLandingLocationChanged)))))
    (:then_1
      (CallFunction|HandleFallTime)
      ;[1 1 1] fetch fall height
      (bind _outxy (Fortnite|EvaluateCurveTableRow (Variables|Default|GetCurveTable_FallHeight)))
      (Variables|Default|SetFallHeight _outxy)
      ;[1 1 1] start falling
      (CallFunction|PlayDrop))))

(fn IsAcceptablePositionForPlacement
  (params (InLocation :type Vector%20%28by%20ref%29 :default "0, 0, 0") (InRotation :type Rotator%20%28by%20ref%29 :default "0, 0, 0") (DecoTool :type Fort%20Deco%20Tool%20Object%20Reference) (bIsCDO :type Boolean))
  (bind (_outhit _returnvalue) (Collision|BoxTraceByChannel (+ InLocation "0, 0,50.000000") (+ InLocation "0, 0,7000.000000") "50.000000,50.000000,50.000000" "0, 0, 0" "TraceTypeQuery10"))
  (return (select (not _returnvalue) "" "NSLOCTEXT(\"\", \"63037FB94900FB1564D17B8D3B141318_Option 00\", \"Landing path obstructed\")") (not _returnvalue)))

(fn LootSpawnPosition
  (return (Math|Vector|vector+vector (Transformation|GetWorldLocation (Variables|Default|GetSupplyDrop_Mesh)) (* (Math|Random|RandomUnitVector) (Math|Random|RandomFloatinRange 20 45)) (Variables|Default|GetLootSpawnOffset))))

(fn MoveOverlappingPawns
  (Networking|SwitchHasAuthority
    (:Authority
      ;[1 1 1] Attempt to move any pawns that are overlapping the supply crate when it lands
      (bind _outactors (Collision|ComponentOverlapActors (Variables|Default|GetCollision) (Transformation|GetWorldTransform (Variables|Default|GetCollision)) (Utilities|Array|MakeArray "ObjectTypeQuery3") "/Script/FortniteGame.FortPawn"))
      (for _array_element _outactors
        (bind _asfort_pawn (Utilities|Casting|CastToFortPawn _array_element))
        (Transformation|Teleport _asfort_pawn (+ (Transformation|GetActorLocation _asfort_pawn) (* (Variables|/Script/Engine.BoxComponent|GetBoxExtent (Variables|Default|GetCollision)) "0, 0,1.000000")) (Transformation|GetActorRotation _asfort_pawn)) ; This should attempt to find a close spot to teleport the Pawn that will not be inside the supply drop
      ))))

(fn OnRep_bLanded
  (bind _result (CallFunction|PlayingReplay))
  (if _result
    (CallFunction|FallTimelineFinished)))

(fn OnRep_bLooted
  (if (Variables|Default|GetbLooted)
    (Networking|CanExecuteCosmeticEvents
      (:True
        (Utilities|FlowControl|DoOnce false
          (:Completed
            (CallFunction|HideandExplodeCrate)))))))

(fn OnRep_NewLocation
  (CallFunction|PlayDropNewTime))

(fn PlayingReplay
  ;[1 1 1 1 11] This is wrapped because I don't like how it appears in native and there might be another, better way to do this.
  (return (CallFunction|IsPlayingAReplay)))

(fn SpawnLoot
  (bind _unlocks (Variables|Default|GetUnlocks))
  (bind (_tierone_5_9d1eb309486dd183fed684b8eeaf54b7 _tiertwo_6_58ab58b441db6edc4fea21abd0c17475 _tierthree_7_f54ff5fc4d92facf83cb69b88043a65f _tierfour_8_6c40909a4282bd0806479a8dd34449eb _tierfive_9_0b433ac44b7dba126b7872bd5ec513da) (Utilities|Struct|BreakSupplyDropUnlocks _unlocks))
  (if (not (Variables|Default|GetbLooted))
    (Utilities|FlowControl|DoOnce false
      (:Completed
        (Networking|SwitchHasAuthority
          (:Authority
            (Variables|Default|SetbLooted true))
          (:Remote
            _))
        (CallFunction|TurnonInteract false)
        (CallFunction|SpawnTierDrops 0)
        (if _tierone_5_9d1eb309486dd183fed684b8eeaf54b7
          (CallFunction|SpawnTierDrops 1)
          (else
            _))
        (if _tiertwo_6_58ab58b441db6edc4fea21abd0c17475
          (CallFunction|SpawnTierDrops 2)
          (else
            _))
        (if _tierthree_7_f54ff5fc4d92facf83cb69b88043a65f
          (CallFunction|SpawnTierDrops 3)
          (else
            _))
        (if _tierfour_8_6c40909a4282bd0806479a8dd34449eb
          (CallFunction|SpawnTierDrops 4)
          (else
            _))
        (if _tierfive_9_0b433ac44b7dba126b7872bd5ec513da
          (CallFunction|SpawnTierDrops 5))))))

(fn SpawnTierDrops
  (params (LootTableIndex :type Integer))
  (bind (_outloottodrop _returnvalue) (Loot|PickLootDrops (Utilities|Array|Get(acopy) (Variables|Default|GetLootTableNames) LootTableIndex) (Loot|GetLootLevel) -1 "(GameplayTags=)"))
  (if _returnvalue
    (for _array_element _outloottodrop
      (bind _asfort_world_item_definition (Utilities|Casting|CastToFortWorldItemDefinition (Class|FortItemFunctionLibrary|BreakFortItemEntry _array_element)))
      (CallFunction|SpawnPickup _asfort_world_item_definition (.Count (Class|FortItemFunctionLibrary|BreakFortItemEntry _array_element)) (Variables|Default|GetMyInteractingPawn) (CallFunction|LootSpawnPosition)))
    (else
      (Development|PrintString (Utilities|String|Append "LogAbility: Failed to spawn tier " (Utilities|String|ToString(Integer) LootTableIndex) " loot from Loot Crate.") false))))

(fn StartDropAudio
  (bind _supplydrop_mesh (Variables|Default|GetSupplyDrop_Mesh))
  (bind _returnvalue (Transformation|GetWorldLocation _supplydrop_mesh))
  (Audio|Components|Audio|FadeIn (Variables|Default|GetFlareAudio) 3 1.0 0.0 "Linear" :!class /Script/Engine.AudioComponent)
  (bind _returnvalue_2 (Audio|SpawnSoundAttached "/Game/Sounds/Foley_Loot/Containers/SupplyDrops/Fort_SupplyDrop_Falling_Loop_Athena_Cue.Fort_SupplyDrop_Falling_Loop_Athena_Cue" (Variables|Default|GetSupplyDrop_Mesh)))
  (bind _output_get (Variables|Default|SetFallingAudio _returnvalue_2))
  (Audio|Components|Audio|FadeIn _output_get 1 1.0 0.0 "Linear" :!class /Script/Engine.AudioComponent)
  (Audio|PlaySoundatLocation "/Game/Sounds/Foley_Loot/Containers/SupplyDrops/Fort_SupplyDrop_Appear_Athena_Cue.Fort_SupplyDrop_Appear_Athena_Cue" _returnvalue)
  (Audio|PlaySoundatLocation "/Game/Sounds/Foley_Loot/Containers/SupplyDrops/Fort_SupplyDrop_Reticle_Appear_Athena_Cue.Fort_SupplyDrop_Reticle_Appear_Athena_Cue" _returnvalue))

(fn ToggleFallingFX
  (params (bToggleOn :type Boolean))
  (Components|Activation|SetActive (Variables|Default|GetTrail) bToggleOn false))

(fn TurnOnInteract
  (params (bNewActorEnableCollision :type Boolean))
  (Collision|SetActorEnableCollision bNewActorEnableCollision)
  (Variables|Default|SetAllow_SupplyDrop_Interact bNewActorEnableCollision))

(fn UpdateDropAudio
  (params (FallPercent :type Float%20%28double-precision%29))
  (bind _fallingaudio (Variables|Default|GetFallingAudio)
    (:then
      (Audio|Parameter|SetFloatParameter _fallingaudio "FallingPercent" FallPercent :!class /Script/AudioExtensions.AudioParameterControllerInterface))
    (:else)))

(fn UpdateFall
  (params (Time :type Float%20%28double-precision%29) (FallPercent :type Float%20%28double-precision%29) (ShouldTeleport :type Boolean))
  (bind _result (CallFunction|PlayingReplay))
  (if (not _result)
    (bind _output_get (Variables|Default|SetFallTime Time)))
  (CallFunction|UpdateLocation _output_get ShouldTeleport)
  (CallFunction|UpdateDropAudio FallPercent))

(fn UpdateLocation
  (params (Time :type Float%20%28double-precision%29) (TeleportMovement :type Boolean))
  (Transformation|SetWorldLocation (Variables|Default|GetInterceptCollision) (Math|Vector|Lerp(Vector) (CallFunction|GetImpactLocation) (CallFunction|GetFallLocation) Time) (not TeleportMovement) TeleportMovement))

(fn ConstructionScript)
