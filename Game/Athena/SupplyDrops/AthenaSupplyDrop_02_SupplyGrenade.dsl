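; Blueprint for the Supply Grenade variant of the Athena supply drop, built on AthenaSupplyDrop_02.
; Its HandleFallTime event reads a fall time from a curve table row, stores its reciprocal as the falling playrate and applies that rate to the inherited falling timeline.
; The construction script only calls the parent, and BeginPlay, overlap and Tick add no logic of their own.
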(event EventBeginPlay
  (|CallParentFunction) ; This node is disabled and will not be called.\nDrag off pins to build functionality.
)

(event Collision|EventActorBeginOverlap
  (params (OtherActor :type Actor%20Object%20Reference)))

(event EventTick
  (params (DeltaSeconds :type Float%20%28single-precision%29)))

(event EventHandleFallTime
  (bind _outxy (Fortnite|EvaluateCurveTableRow (Variables|Default|GetCurveTable_FallTime)))
  (bind _output_get (Variables|Default|SetFallingPlayrate (/ 1.0 _outxy)))
  (Components|Timeline|SetPlayRate (Variables|AthenaSupplyDrop_02|GetFallingTimeline) _output_get))

(fn ConstructionScript
  (|CallParentFunction))
