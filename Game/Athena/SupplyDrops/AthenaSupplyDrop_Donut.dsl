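; Donut variant of the Athena supply drop, whose one piece of custom logic drives its gadget animation on landing.
; On landing, it casts its child actor to the Donut balloon and reads that balloon's gadget skeletal mesh.
; It then calls OnLanded on the mesh's post-process anim instance, rather than the main anim instance.
; BeginPlay, overlap and tick are empty stubs, and the construction script only calls the parent.
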(event EventBeginPlay
  (|CallParentFunction) ; This node is disabled and will not be called.\nDrag off pins to build functionality.
)

(event Collision|EventActorBeginOverlap
  (params (OtherActor :type Actor%20Object%20Reference)))

(event EventTick
  (params (DeltaSeconds :type Float%20%28single-precision%29)))

(event EventOnLanding
  (bind _asathena_supply_drop_balloon_donut (Utilities|Casting|CastToAthenaSupplyDropBalloon_Donut (Variables|/Script/Engine.ChildActorComponent|GetChildActor (Variables|Default|GetChildActor))))
  (bind _assupply_drop_donut_gadget_anim_bp (Utilities|Casting|CastToSupply_Drop_Donut_Gadget_AnimBP (Components|SkeletalMesh|GetPostProcessInstance (Variables|/Game/Athena/SupplyDrops/AthenaSupplyDropBalloon_Donut.AthenaSupplyDropBalloon_Donut_C|GetSupply_Drop_Donut_Gadget _asathena_supply_drop_balloon_donut))))
  (Class|SupplyDropDonutGadgetAnimBP|OnLanded _assupply_drop_donut_gadget_anim_bp))

(fn ConstructionScript
  (|CallParentFunction))
