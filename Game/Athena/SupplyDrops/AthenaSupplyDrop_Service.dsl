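; Athena supply drop crate that descends under a balloon to a ground reticle, then spawns loot and any configured vehicle when opened or destroyed.
; The server sets the angled descent; balloon pops switch to gravity, and lost landing support restarts falling.
; Replicated state drives effects and snaps stuck clients to the landing spot.
; Performance measures include deferred client collision sweeps, conditional water overlaps, distance-gated audio and one loot pickup per frame.
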(event Custom|LandingInterrupted
  (CallFunction|PlayLandingFX false))

(event Custom|PhysicsActivated
  (params (Component :type Actor%20Component%20Object%20Reference) (bReset :type Boolean))
  ;[1 1 1] BeginPlay - Initialize fall, balloon, and timers / callbacks / set up location and direction
  (Actor|DestroyActor (Variables|/Script/Engine.ChildActorComponent|GetChildActor (Variables|Default|GetChildActor)))
  (Utilities|FlowControl|DoOnce false
    (:Completed
      (Variables|Default|SetLandInterrupted true)
      (CallFunction|LandingInterrupted))))

(event EventBeginPlay
  (bind _childactor (Variables|Default|GetChildActor))
  (bind _childactor_2 (Variables|/Script/Engine.ChildActorComponent|GetChildActor _childactor))
  (bind _cratestaticmesh (Variables|Default|GetCrateStaticMesh))
  (bind _linktoactorcomponent (Variables|FortAthenaSupplyDrop|GetLinkToActorComponent))
  (bind _childactor_3 (Variables|Default|GetChildActor))
  (bind _childactor_4 (Variables|/Script/Engine.ChildActorComponent|GetChildActor _childactor_3))
  (bind _childactor_5 (Variables|Default|GetChildActor))
  (bind _childactor_6 (Variables|/Script/Engine.ChildActorComponent|GetChildActor _childactor_5))
  ;[1 1 1] BeginPlay - Initialize fall, balloon, and timers / callbacks / set up location and direction
  (Utilities|FlowControl|Sequence
    (:then_0
      (bind _returnvalue (BuildingActor|PhysicsSimulationSupport|GetPhysicsObjectComponent self))
      (Utilities|IsValid _returnvalue
        (:Is%20Valid
          (Components|Activation|BindEventtoOnComponentActivated _returnvalue (delegate Custom|PhysicsActivated)))))
    (:then_1
      (Networking|SwitchHasAuthority
        (:Authority
          (EventDispatchers|BindEventtoOnLinkedPhysicsObjectWake _linktoactorcomponent (EventDispatchers|CreateEvent :!event Custom|HandleLinkedPhysicsObjectAwake))
          (EventDispatchers|BindEventtoOnLinkedActorDestroyed _linktoactorcomponent (EventDispatchers|CreateEvent :!event Custom|HandleLinkedActorDestroyed))
          (EventDispatchers|BindEventtoOnLinkedActorDisconnected _linktoactorcomponent (EventDispatchers|CreateEvent :!event Custom|HandleLinkedActorDisconnected))
          (Utilities|Time|SetTimerbyEvent (EventDispatchers|CreateEvent :!event Custom|Multicast_ApplyGravityForFall) (Variables|Default|GetInitialFallDelay))
          ;;[1 1 1] [QuestSetup] Added by Nicholas.Lacelle | Apply GE that has the quest tag Athena.Quests."SupplyDropIsFalling"
          (Utilities|IsValid _childactor_4
            (:Is%20Valid
              (GameplayEffects|ApplyGameplayEffectToSelf :self (Ability|GetAbilitySystemComponent _childactor_4) :GameplayEffectClass "/Game/Athena/Items/QuestInteractables/Quest_GEs/GE_Athena_Quests_SupplyDrop.GE_Athena_Quests_SupplyDrop_C" :Level 0.0))))))
    (:then_2
      (CallFunction|InitFallHeightandSpeed)
      (Utilities|IsValid _childactor_6
        (:Is%20Valid
          (bind _asathena_supply_drop_balloon (Utilities|Casting|CastToAthenaSupplyDropBalloon _childactor_6))
          (Transformation|AttachComponentToComponent (Variables|/Script/Engine.Actor|GetRootComponent _childactor_6) (Variables|Default|GetCrateStaticMesh) "None" "KeepWorld" "SnapToTarget" "KeepWorld" true)
          (Default|AssignOnBalloonDestroyed _asathena_supply_drop_balloon (EventDispatchers|CreateEvent :!event BalloonDestroyed)))))
    (:then_3
      ;;[1 1 1] Set up initial position directly above the ground
      (Networking|SwitchHasAuthority
        (:Authority
          (Transformation|Teleport (Math|Vector|MakeVector (.x (Transformation|GetActorLocation)) (.y (Transformation|GetActorLocation)) (+ (.z (Transformation|GetActorLocation)) (Variables|Default|GetFallHeight))) (Math|Rotator|SelectRotator (Variables|Default|GetOverrideRotation) (Transformation|GetActorRotation) (Variables|Default|GetbOverrideRotation))))))
    (:then_4
      ;;[1 1 1] Steup the reticle with the initial straight down fall direction. We purposely allow non cosmetic in reticle location so we can use that location for moving the drop along its offset later
      (CallFunction|UpdateReticleLocation)
      (CallFunction|StartUpdateReticleTimer)
      (Networking|CanExecuteCosmeticEvents
        (:True
          (Audio|PlaySoundatLocation (Variables|Sound|GetSound_Appear) (Transformation|GetActorLocation))
          (Audio|PlaySound2D (Variables|Sound|GetSound%20Appear%20Dialogue))
          (CallFunction|TimerCheckPlayerDistance)
          (if (Utilities|IsValid (Variables|Default|GetReticle_Niagara) :!class /Script/Engine.KismetSystemLibrary)
            (Audio|Components|Audio|FadeIn (Variables|Default|GetFlareAudio) 3 1.0 0.0 "Linear" :!class /Script/Engine.AudioComponent)))))
    (:then_5
      ;;[1 1 1] Add whatever offset the drop wants (for coming in at an angle) and resetting the direction
      (Networking|SwitchHasAuthority
        (:Authority
          (Transformation|Teleport (+ (Transformation|GetActorLocation) (Math|Vector|MakeVector (Math|Random|RandomFloatinRange (Variables|Default|GetDescentOffsetMin) (Variables|Default|GetDescentOffsetMax)) (Math|Random|RandomFloatinRange (Variables|Default|GetDescentYOffsetMin) (Variables|Default|GetDescentYOffsetMax)))) (Transformation|GetActorRotation))
          (Variables|Default|SetFallDirection (Math|Vector|Normalize (- (Variables|Default|GetGroundLocation) (Transformation|GetActorLocation)))))))
    (:then_6
      ;;[1 0.9219 0.179] Anniversary Event:  When Calendar Event detected, change materials accordingly.  Must update Event Name and Material Variables with every new Anniversary Event as needed
      (bind _biseventactive (CalendarEvents|GetCalendarEventInformation (Variables|Anniversary|GetAnniversary%20Event%20Name)))
      (if _biseventactive
        (Development|PrintString "BIRTHDAY ACTIVE" true true "(R=1.000000,G=0.730108,B=0.000000,A=1.000000)" 2e+01)
        (Utilities|IsValid _cratestaticmesh
          (:Is%20Valid
            (Rendering|Material|SetMaterial _cratestaticmesh 0 (Variables|Anniversary|GetAnniversary%20Crate%20Material))
            (Utilities|IsValid _childactor_2
              (:Is%20Valid
                (bind _asathena_supply_drop_balloon_2 (Utilities|Casting|CastToAthenaSupplyDropBalloon _childactor_2))
                (Rendering|Material|SetMaterial (Variables|/Script/FortniteGame.BuildingSMActor|GetStaticMeshComponent _asathena_supply_drop_balloon_2) 0 (Variables|Anniversary|GetAnniversary%20Balloon%20Material))))))))
    (:then_7
      ;;[1 1 1] Client: optionally disable client collision sweeps when movement starts. They are turned on later under important conditions (balloon popped, close to landing, etc)
      ;; Controlled by cvar: "Fort.SupplyDropClientCollisionCheckReductionsEnabled"
      (Networking|SwitchHasAuthority
        (:Remote
          (if (CallFunction|IsClientCollisionReductionEnabled)
            (Variables|/Script/Engine.ProjectileMovementComponent|SetbSweepCollision false (Variables|Default|GetFortProjectileMovement))))))))

(event Custom|StopUpdateReticleTimer
  ;[1 1 1] BeginPlay - Initialize fall, balloon, and timers / callbacks / set up location and direction
  ;;[1 1 1] Stop UpdateReticleLocation timer
  (Utilities|Time|ClearandInvalidateTimerbyHandle (Variables|Default|GetUpdateReticleTimer)))

(event Custom|StartUpdateReticleTimer
  ;[1 1 1] BeginPlay - Initialize fall, balloon, and timers / callbacks / set up location and direction
  ;;[1 1 1] Fire UpdateReticleLocation on a timer
  (bind _returnvalue (Utilities|Time|SetTimerbyEvent (EventDispatchers|CreateEvent :!event UpdateReticleLocation) (Variables|Default|GetReticleUpdateInterval) true false (Variables|Default|GetReticleUpdateInterval) (* (Variables|Default|GetReticleUpdateInterval) 0.5)))
  (Variables|Default|SetUpdateReticleTimer _returnvalue))

(event Custom|TimerCheckPlayerDistance
  (bind _fallingaudio (Variables|Default|GetFallingAudio))
  (bind _returnvalue (Audio|Components|Audio|IsPlaying _fallingaudio))
  ;[1 1 1] Timer to check if fall audio should play or not. When not playing, auto-detaches to help with movement perf.
  (Utilities|IsValid (Variables|/Script/Engine.AudioComponent|GetSound (Variables|Default|GetFallingAudio))
    (:Is%20Valid
      (bind _returnvalue_2 (Utilities|Time|SetTimerbyEvent (delegate Custom|TimerCheckPlayerDistance) (Math|Float|Max(Float) 0.5 (Variables|Default|GetPlayerDistanceCheckTimeLoop))))
      (Variables|Default|SetFallingAudioCheckTimerHandle _returnvalue_2)
      (bind _returnvalue_3 (Audio|AreAnyListenersWithinRange (Transformation|GetActorLocation) (Variables|/Script/Engine.SoundBase|GetMaxDistance (Variables|/Script/Engine.AudioComponent|GetSound (Variables|Default|GetFallingAudio)))))
      (if _returnvalue_3
        (if (not _returnvalue)
          (CallFunction|BeginFallAudioTimeline))
        (elif _returnvalue
          (CallFunction|StopFallAudioTimeline))))))

(event Custom|BeginFallAudioTimeline
  ;[1 1 1] Falling audio timeline
  (Audio|Components|Audio|FadeIn (Variables|Default|GetFallingAudio) 1.0 1.0 0.0 "Linear" :!class /Script/Engine.AudioComponent))

(event Custom|DestroyGroundComponents
  (CallFunction|StopUpdateReticleTimer)
  (Utilities|FlowControl|Delay 5.0)
  (Components|DestroyComponent (Variables|Default|GetReticleAudio))
  (if (not (Variables|Default|GetbGroundFXDisabled))
    (Variables|Default|SetbGroundFXDisabled true)))

(event Custom|StopFallAudioTimeline
  ;[1 1 1] Falling audio timeline
  (Audio|Components|Audio|FadeOut (Variables|Default|GetFallingAudio) 0.25 0.0 "Linear" :!class /Script/Engine.AudioComponent))

(event OnProjectileStop(FortProjectileMovement)
  (params (ImpactResult :type Hit%20Result%20Structure%20%28by%20ref%29))
  ;[1 1 1] If the supply drop falls and lands on a building actor again, bind to fall if it goes away
  (CallFunction|HandleLanding ImpactResult))

(event Custom|SetupQuestData
  (bind _self self)
  (bind _returnvalue (Ability|GetAbilitySystemComponent _self))
  (bind _childactor (Variables|Default|GetChildActor))
  (bind _childactor_2 (Variables|/Script/Engine.ChildActorComponent|GetChildActor _childactor))
  (Utilities|FlowControl|Sequence
    (:then_0
      ;[1 1 1] [QuestSetup] Added by Nicholas.Lacelle | Once the supply drop has landed we want to remove the tag so palyers don't get quest credit from a still supply drop
      (Utilities|IsValid _childactor_2
        (:Is%20Valid
          (GameplayEffects|RemoveActiveEffectswithGrantedTags (Ability|GetAbilitySystemComponent _childactor_2) "(GameplayTags=((TagName=\"Athena.Quests.SupplyDropIsFalling\")))"))))
    (:then_1
      ;[1 1 1] QuestSetup: Add tag upon landing
      (Utilities|IsValid _returnvalue
        (:Is%20Valid
          (GameplayEffects|ApplyGameplayEffectToSelf :self _returnvalue :GameplayEffectClass "/Game/Athena/Items/Quests/QuestAbilities/QuestGEs/GE_Quest_Search_SupplyDrop_AfterLanding.GE_Quest_Search_SupplyDrop_AfterLanding_C" :Level 0.0))))))

(event Custom|HandleLanding
  (params (ImpactResult :type Hit%20Result%20Structure))
  ;[1 1 1] If the supply drop falls and lands on a building actor again, bind to fall if it goes away
  (CallFunction|HandleProjectileMovementStop ImpactResult)
  (CallFunction|OnLanding)
  (CallFunction|DestroyGroundComponents)
  (CallFunction|SetupQuestData)
  (bind _asathena_supply_drop_balloon (Utilities|Casting|CastToAthenaSupplyDropBalloon (Variables|/Script/Engine.ChildActorComponent|GetChildActor (Variables|Default|GetChildActor))))
  (Marking|HandleSupplyDropMovementStopped _asathena_supply_drop_balloon ImpactResult))

(event Custom|OnLandingLocationChanged
  ;[1 1 1] Event fired when actor supporting the supply drop has died or moved
  (if (Math|Vector|NotEqual(Vector) (Variables|Default|GetLandingLocation) "0, 0, 0" 0.1)
    (Variables|Default|SetLandingLocation "0, 0, 0")
    (Utilities|FlowControl|Delay 0.05) ; If clients start falling too early, they hit the not-yet-destroyed structure
    (CallFunction|MulticastApplyGravityforFall)
    (CallFunction|UpdateReticleLocation)
    (else
      (CallFunction|UpdateReticleLocation))))

(event Custom|Multicast_ApplyGravityForFall
  ;[1 1 1] Applies the Gravity client & server so the supply drop falls when its support structure goes away
  (CallFunction|InitFall))

(event Interaction|EventBlueprintOnInteract
  (params (InteractingPawn :type Fort%20Pawn%20Object%20Reference) (InteractionBeingAttempted :type EInteractionBeingAttempted%20Enum :default "FirstInteraction"))
  (bind _returnvalue (Ability|GetAbilitySystemComponent InteractingPawn))
  ;[1 1 1] Interact
  (Utilities|FlowControl|Sequence
    (:then_0
      (Variables|Default|SetMyInteractingPawn InteractingPawn)
      (if (not (Variables|Default|GetAutoOpen))
        (CallFunction|OpenSupplyDrop)
        ;[1 1 1] Update quests
        (GameplayEffects|ApplyGameplayEffectToSelf :self _returnvalue :GameplayEffectClass "/Game/Athena/Items/QuestInteractables/Quest_GEs/GE_Athena_Quests_PostSupplyDropInteract.GE_Athena_Quests_PostSupplyDropInteract_C" :Level 0.0)
        (GameplayEffects|ApplyGameplayEffectToTarget :self _returnvalue :GameplayEffectClass "/Game/Athena/Items/Quests/Feats/GEs/GE_Feat_S11_RecentlyOpenedSupplyDrop.GE_Feat_S11_RecentlyOpenedSupplyDrop_C" :Target _returnvalue :Level 0.0)))
    (:then_1
      (Variables|Default|SetMyInteractingPawn 0))))

(event EventOnDeathServer
  (params (Damage :type Float%20%28single-precision%29) (DamageTags :type Gameplay%20Tag%20Container%20Structure%20%28by%20ref%29) (Momentum :type Vector :default "0, 0, 0") (HitInfo :type Hit%20Result%20Structure%20%28by%20ref%29) (InstigatedBy :type Controller%20Object%20Reference) (DamageCauser :type Actor%20Object%20Reference) (EffectContext :type Gameplay%20Effect%20Context%20Handle%20Structure))
  ;[1 1 1] OnDeathServer
  (Utilities|FlowControl|DoOnce false
    (:Completed
      (Collision|SetActorEnableCollision false)
      (Actor|SetLifeSpan 2.0)
      (CallFunction|OpenSupplyDrop))))

(event EventOnDeathPlayEffects
  (params (Damage :type Float%20%28single-precision%29) (DamageTags :type Gameplay%20Tag%20Container%20Structure%20%28by%20ref%29) (Momentum :type Vector :default "0, 0, 0") (HitInfo :type Hit%20Result%20Structure%20%28by%20ref%29) (InstigatedBy :type Fort%20Pawn%20Object%20Reference) (DamageCauser :type Actor%20Object%20Reference) (EffectContext :type Gameplay%20Effect%20Context%20Handle%20Structure))
  (bind _cratestaticmesh (Variables|Default|GetCrateStaticMesh))
  (bind _returnvalue (Transformation|GetWorldTransform _cratestaticmesh))
  (bind _location (.location _returnvalue))
  ;[1 1 1] OnDeathPlayEffects - when the supply drop is destroyed, rather than looted
  (Networking|CanExecuteCosmeticEvents
    (:True
      (if (not (Variables|Default|GetOpened))
        (Niagara|SpawnSystematLocation "/Game/Effects/Fort_Effects/World/Destruction/Base_Effects/NS_General_Deco_Destruction_Small_Converted.NS_General_Deco_Destruction_Small_Converted" _location (.rotation _returnvalue))
        (Audio|PlaySoundatLocation (Variables|Sound|GetSound_Destroyed) _location)
        (Rendering|SetActorHiddenInGame true)))))

(event Custom|Play%20Looted%20Dialogue
  (Audio|PlaySound2D (Variables|Sound|GetSound%20Looted%20Inf)))

(event OnEnterWaterEvent(FortWaterInteraction)
  (params (WaterBody :type Fort%20Water%20Body%20Actor%20Object%20Reference) (WaterInteractionComponent :type Fort%20Water%20Interaction%20Component%20Object%20Reference) (bIsFirstBody :type Boolean))
  (CallFunction|PlayLandingFX false)
  (CallFunction|PlayLandingFXinWater)
  (CallFunction|DestroyGroundComponents)
  (CallFunction|OnLanding)
  (CallFunction|SetupQuestData)
  (if (Variables|Default|GetBalloonPopped)
    ;[1 1 1] Set velocity to 0 to avoid getting stuck in ground
    (Variables|/Script/Engine.MovementComponent|SetVelocity "0, 0, 0" (Variables|Default|GetFortProjectileMovement))
    (else
      ;[1 1 1] Set movement component's updated component to NULL to prevent floating down river
      (Components|Movement|SetUpdatedComponent (Variables|Default|GetFortProjectileMovement) 0))))

(event Custom|OnDelayedSpawnLoot
  (bind _delayedspawnloot (Variables|Default|GetDelayedSpawnLoot))
  (bind _output (Utilities|Array|Get(aref) _delayedspawnloot))
  (bind (_itemdefinition _itemguid _count) (Class|FortItemFunctionLibrary|BreakFortItemEntry _output))
  (bind _delayedspawnloot_2 (Variables|Default|GetDelayedSpawnLoot))
  (if (!= (Utilities|Array|Length _delayedspawnloot) 0)
    (bind _asfort_world_item_definition (Utilities|Casting|CastToFortWorldItemDefinition _itemdefinition))
    (bind _returnvalue (CallFunction|SpawnPickup _asfort_world_item_definition _count (Variables|Default|GetMyInteractingPawn) (CallFunction|GetLootSpawnLocation)))
    ;[1 1 1] applying mod set functions to weapon ids
    (bind _asfort_weapon_item_definition (Utilities|Casting|CastToFortWeaponItemDefinition _asfort_world_item_definition)
      (:then
        (bind _returnvalue_2 (Class|FortWeaponModFunctionLibrary|GetRandomWeaponModSetData _asfort_weapon_item_definition))
        (Class|FortWeaponModFunctionLibrary|ApplyModSettoPickup _returnvalue_2 _returnvalue))
      (:CastFailed
        _))
    (CallFunction|OnLootSpawned)
    (Utilities|Array|RemoveIndex _delayedspawnloot_2)
    (if (!= (Utilities|Array|Length _delayedspawnloot_2) 0)
      (Utilities|FlowControl|Delay 0.0) ; Delay until the next frame
      (CallFunction|OnDelayedSpawnLoot)
      (else
        (Actor|DestroyActor)))
    (else
      (Actor|DestroyActor))))

(event EventOnTouchedFillFloor
  (params (FillFloorActor :type Athena%20Fill%20Floor%20Object%20Reference))
  (Utilities|Casting|CastToAthenaLavaFloor FillFloorActor)
  (CallFunction|OpenSupplyDrop))

(event Custom|OnLanding)

(event Custom|HandleLinkedPhysicsObjectAwake
  (params (PhysicsObject :type Actor%20Object%20Reference))
  ;[1 1 1 0.5] LinkToActorComponent handles structures the supply drop is resting on
  (CallFunction|OnLandingLocationChanged))

(event Custom|HandleLinkedActorDestroyed
  (params (DamageCauser :type Actor%20Object%20Reference))
  ;[1 1 1 0.5] LinkToActorComponent handles structures the supply drop is resting on
  (CallFunction|OnLandingLocationChanged))

(event Custom|HandleLinkedActorDisconnected
  ;[1 1 1 0.5] LinkToActorComponent handles structures the supply drop is resting on
  (CallFunction|OnLandingLocationChanged))

(fn AlmostLanded
  (if (not (Variables|Default|GetBalloonPopped))
    (Audio|SpawnSoundAttached (Variables|Sound|GetSound_AlmostLanded) (Variables|Default|GetCrateStaticMesh))))

(fn Analytics
  (params (ItemsToSpawn :type Array%20of%20Fort%20Item%20Entry%20Structures))
  (Utilities|IsValid (Variables|Default|GetMyInteractingPawn)
    (:Is%20Valid
      (bind _asfort_player_controller (Utilities|Casting|CastToFortPlayerController (Pawn|GetController (Variables|Default|GetMyInteractingPawn))))
      (Analytics|SendSpawnLootAnalytics _asfort_player_controller (Variables|GetItemsToSpawn)))))

(fn BalloonDestroyed
  (Networking|SwitchHasAuthority
    (:Authority
      (Variables|Default|SetBalloonPopped true)
      (CallFunction|MulticastApplyGravityforFall))))

(fn BlueprintCanInteract
  (params (InteractingPawn :type Fort%20Pawn%20Object%20Reference) (InteractionBeingAttempted :type EInteractionBeingAttempted%20Enum :default "FirstInteraction") (InteractionType :type TInteractionType%20Enum :default "IT_NoInteraction"))
  (return (and (Math|Boolean|ORBoolean (Math|Vector|NotEqual(Vector) (Variables|Default|GetLandingLocation) "0, 0, 0" 0.1) (Water|IsinWater (Variables|Default|GetFortWaterInteraction)) (<= (Math|Vector|VectorLengthSquared (Variables|/Script/Engine.MovementComponent|GetVelocity (Variables|Default|GetFortProjectileMovement))) 0.01)) (not (Variables|Default|GetOpened)))))

(fn BlueprintGetInteractionString
  (params (InteractingPawn :type Fort%20Pawn%20Object%20Reference) (InteractionBeingAttempted :type EInteractionBeingAttempted%20Enum :default "FirstInteraction"))
  (return (Variables|GetBuildingActorInteractionText)))

(fn DeactivateGroundFX
  (bind _reticle_niagara (Variables|Default|GetReticle_Niagara)
    (:then
      (Components|Activation|Deactivate _reticle_niagara))
    (:else
      _))
  (bind _reticleaudio (Variables|Default|GetReticleAudio)
    (:then
      (Audio|Components|Audio|FadeOut _reticleaudio 0.2 0.0 "Linear" :!class /Script/Engine.AudioComponent))
    (:else
      _))
  (CallFunction|DisableFlare))

(fn DisableFlare
  (Components|Activation|Deactivate (Variables|Default|GetFlare_Niagara))
  (Audio|Components|Audio|FadeOut (Variables|Default|GetFlareAudio) 3.0 0.0 "Linear" :!class /Script/Engine.AudioComponent)
  (Utilities|Time|ClearTimerbyFunctionName 0 "DisableFlare"))

(fn ForceSetLocationOnLanding
  (bind _fortprojectilemovement (Variables|Default|GetFortProjectileMovement))
  ;[1 1 1] Force client supply drop to server location if it hasn't stopped due to parts of level not being streamed in
  (Networking|SwitchHasAuthority
    (:Remote
      (if (not (Math|Vector|Equal(Vector) (Variables|Default|GetLandingLocation) "0, 0, 0" 0.1))
        (Transformation|Teleport (Variables|Default|GetLandingLocation) (Transformation|GetActorRotation))
        (Variables|/Script/Engine.ProjectileMovementComponent|SetProjectileGravityScale 0.0 _fortprojectilemovement)
        (Variables|/Script/Engine.MovementComponent|SetVelocity "0, 0, 0" _fortprojectilemovement)
        (Components|Movement|SetUpdatedComponent _fortprojectilemovement 0) ; Stop ticking
      ))))

(fn GetLootSpawnLocation
  (return (Transformation|GetActorLocation)))

(fn HandleProjectileMovementStop
  (params (ImpactResult :type Hit%20Result%20Structure))
  (bind (_bblockinghit _binitialoverlap _time _distance _location _impactpoint _normal _impactnormal _physmat _hitactor) (Collision|BreakHitResult ImpactResult))
  (Variables|/Script/Engine.ProjectileMovementComponent|SetbSweepCollision true (Variables|Default|GetFortProjectileMovement))
  (Utilities|Casting|CastToFortProjectileBase _hitactor
    (:then
      _)
    (:CastFailed
      (Utilities|Casting|CastToFortRemoteControlledPawnAthena _hitactor
        (:then
          _)
        (:CastFailed
          (Utilities|Casting|CastToFortPickupAthena _hitactor
            (:then
              _)
            (:CastFailed
              (Utilities|Casting|CastToPawn _hitactor
                (:then
                  (Utilities|Casting|CastToFortMountedTurret _hitactor
                    (:then
                      ;[1 1 1] If we land on a mounted Turret then go ahead and open
                      (Utilities|FlowControl|Sequence
                        (:then_0
                          (Utilities|Time|ClearTimerbyFunctionName 0 "Multicast_ApplyGravityForFall"))
                        (:then_1
                          (Networking|SwitchHasAuthority
                            (:Authority
                              (Variables|Default|SetLandingLocation (Transformation|GetActorLocation)))))
                        (:then_2
                          (CallFunction|PlayLandingFX true))
                        (:then_3
                          (CallFunction|OpenSupplyDrop))))
                    (:CastFailed
                      _)))
                (:CastFailed
                  (Utilities|Casting|CastToFortPawn (Actor|GetOwner _hitactor)
                    (:then
                      _)
                    (:CastFailed
                      ;[1 1 1] Handle landing
                      (Utilities|FlowControl|Sequence
                        (:then_0
                          (Utilities|Time|ClearTimerbyFunctionName 0 "Multicast_ApplyGravityForFall"))
                        (:then_1
                          (Networking|SwitchHasAuthority
                            (:Authority
                              (Variables|Default|SetLandingLocation (Transformation|GetActorLocation)))))
                        (:then_2
                          (CallFunction|PlayLandingFX true))
                        (:then_3
                          (if (Variables|Default|GetAutoOpen)
                            (CallFunction|OpenSupplyDrop))))))))
              ;[1 1 1] If we stopped due to hitting a pawn / vehicle, after some delay attempt to resume falling
              (Networking|SwitchHasAuthority
                (:Authority
                  (Utilities|Time|SetTimerbyEvent (EventDispatchers|CreateEvent :!event Multicast_ApplyGravityForFall) 2.0)))))))))
  ;[1 1 1] If we just hit a projectile, or pickup, resume falling
  (Networking|SwitchHasAuthority
    (:Authority
      (Utilities|Time|SetTimerbyEvent (EventDispatchers|CreateEvent :!event Multicast_ApplyGravityForFall) 1e-05))))

(fn InitFall
  (bind _fortprojectilemovement (Variables|Default|GetFortProjectileMovement))
  (Networking|ForceNetUpdate)
  (Components|Movement|SetUpdatedComponent _fortprojectilemovement (Variables|Default|GetCrateCollision)) ; Prompts FortProjectileMovement to start ticking and check falling again.
  (Utilities|Time|ClearTimerbyFunctionName 0 "ForceSetLocationOnLanding")
  (if (Variables|Default|GetBalloonPopped)
    (Variables|/Script/Engine.ProjectileMovementComponent|SetProjectileGravityScale 1.0 _fortprojectilemovement)
    (else
      (Variables|/Script/Engine.MovementComponent|SetVelocity (* (Variables|Default|GetFallDirection) (Math|Conversions|ToVector(Float) (Variables|Default|GetFallSpeed))) _fortprojectilemovement))))

(fn InitFallHeightAndSpeed
  (if (Variables|Default|GetBlingDrop)
    (Variables|Default|SetFallHeight (Variables|Default|GetFallHeight))
    (else
      (bind (_minspawnheight _maxspawnheight) (SpawnData|GetSupplyDropSpawnData))
      (Variables|Default|SetFallHeight (Math|Random|RandomFloatinRange _minspawnheight _maxspawnheight))))
  (bind (_minspeed _maxspeed) (SpawnData|GetSupplyDropSpeedData))
  (bind _output_get (Variables|Default|SetFallSpeed (Math|Random|RandomFloatinRange _minspeed _maxspeed)))
  (LocalVariables|SetFallTime (/ (Variables|Default|GetFallHeight) _output_get))
  (Networking|CanExecuteCosmeticEvents
    (:True
      (Utilities|Time|SetTimerbyEvent (EventDispatchers|CreateEvent :!event DisableFlare) (* (LocalVariables|GetFallTime) 0.7))
      (Utilities|Time|SetTimerbyEvent (EventDispatchers|CreateEvent :!event AlmostLanded) (* (LocalVariables|GetFallTime) 0.95)))))

(fn OnRep_BalloonPopped
  (Audio|PlaySoundatLocation "/Game/Sounds/Foley_Loot/Containers/SupplyDrops/Fort_SupplyDrop_Balloon_Explode_Cue.Fort_SupplyDrop_Balloon_Explode_Cue" (Transformation|GetWorldLocation (Variables|Default|GetCrateStaticMesh)))
  (Variables|/Script/Engine.ProjectileMovementComponent|SetbSweepCollision true (Variables|Default|GetFortProjectileMovement))
  (if (Math|Vector|Equal(Vector) (Variables|Default|GetLandingLocation) "0, 0, 0" 0.1)
    (Effects|SpawnEmitterAttached :EmitterTemplate (Variables|Default|GetFallTrailFX) :AttachToComponent (Variables|Default|GetCrateStaticMesh) :AttachPointName "None" :Scale "1.000000,1.000000,1.000000" :LocationType "SnapToTarget" :bAutoDestroy true :PoolingMethod "None" :bAutoActivate true)))

(fn OnRep_bGroundFXDisabled
  (CallFunction|DeactivateGroundFX)
  (CallFunction|DestroyGroundComponents))

(fn OnRep_LandingLocation
  (Networking|SwitchHasAuthority
    (:Remote
      (Variables|/Script/Engine.ProjectileMovementComponent|SetbSweepCollision true (Variables|Default|GetFortProjectileMovement))
      ;[1 1 1] Force client supply drop to server location after a slight delay if we haven't stopped due to parts of level not being streamed in
      (if (Math|Vector|Equal(Vector) (Variables|Default|GetLandingLocation) "0, 0, 0" 0.1)
        (Utilities|Time|ClearTimerbyFunctionName 0 "ForceSetLocationOnLanding")
        (else
          (Utilities|Time|SetTimerbyFunctionName 0 "ForceSetLocationOnLanding" 0.5))))))

(fn OnRep_Opened
  (bind _cratestaticmesh (Variables|Default|GetCrateStaticMesh))
  (bind _returnvalue (Transformation|GetWorldTransform _cratestaticmesh))
  (bind _location (.location _returnvalue))
  (if (Variables|Default|GetOpened)
    (Networking|CanExecuteCosmeticEvents
      (:True
        (Development|SetHiddeninGame (Variables|Default|GetCrateStaticMesh) true true)
        ; FORT-781967 to convert the old cascade FX to Niagara
        (Effects|Components|ParticleSystem|SpawnFXAtLocation (Variables|Looted|GetLootedFX_Niagara) _location (.rotation _returnvalue) "1.000000,1.000000,1.000000" true "AutoRelease") ; FORT-781967 to convert the old cascade FX to Niagara
        (Audio|SpawnSoundatLocation (Variables|Looted|GetSound_Looted) _location)
        (if (Variables|Default|GetInfDrop)
          (Utilities|Time|SetTimerbyEvent (EventDispatchers|CreateEvent :!event Play%20Looted%20Dialogue) 1.0))))))

(fn OnSetCustomDepthStencilValue
  (params (PrimComponents :type Array%20of%20Primitive%20Component%20Object%20References) (bUseCustomDepth :type Boolean) (StencilValue :type Integer))
  (for _array_element PrimComponents
    (Rendering|SetCustomDepthStencilValue _array_element (Variables|GetStencilValue))
    (if (Utilities|Equal(Object) _array_element (Variables|Default|GetCrateStaticMesh))
      (if (Variables|GetbUseCustomDepth)
        (Rendering|SetRenderCustomDepth _array_element true))
      (else
        (Rendering|SetRenderCustomDepth _array_element (Variables|GetbUseCustomDepth)))))
  (return true))

(fn OpenSupplyDrop
  (bind _childactor (Variables|Default|GetChildActor))
  (bind _childactor_2 (Variables|/Script/Engine.ChildActorComponent|GetChildActor _childactor))
  (Networking|SwitchHasAuthority
    (:Authority
      (if (not (Variables|Default|GetOpened))
        (Variables|Default|SetOpened true)
        (Networking|ForceNetUpdate)
        (CallFunction|SpawnLoot)
        (Actor|SetLifeSpan (Variables|Default|GetLifespanOnOpen))
        (Utilities|IsValid _childactor_2
          (:Is%20Valid
            (Actor|DestroyActor _childactor_2)))))))

(fn PlayLandingFX
  (params (bSpawnVFX :type Boolean :default true))
  (bind _cratecollision (Variables|Default|GetCrateCollision))
  (bind _returnvalue (Transformation|GetWorldTransform _cratecollision))
  (bind _location (.location _returnvalue))
  (CallFunction|StopUpdateReticleTimer)
  (Networking|CanExecuteCosmeticEvents
    (:True
      (Utilities|FlowControl|Sequence
        (:then_0
          ;[1 1 1] Turn off Reticle and Flare FX and sounds
          (CallFunction|DeactivateGroundFX))
        (:then_1
          ;[1 1 1] Sound, FX, and Postprocess
          (if (not (Variables|Default|GetLandInterrupted))
            (Utilities|FlowControl|Sequence
              (:then_0
                (Utilities|FlowControl|DoOnce false
                  (:Completed
                    (Audio|PlaySoundatLocation (Variables|Sound|GetSound_Landing) _location))))
              (:then_1
                (if (Variables|GetbSpawnVFX)
                  (Effects|Components|ParticleSystem|SpawnFXAtLocation (Variables|Default|GetLandingFX) (Math|Vector|MakeVector (.x _location) (.y _location) (- (.z _location) (.z (Variables|/Script/Engine.BoxComponent|GetBoxExtent (Variables|Default|GetCrateCollision)))))))))))
        (:then_2
          ;[1 1 1] Stop timer / timeline for fall audio
          (Utilities|Time|ClearTimerbyFunctionName 0 "AlmostLanded")
          (CallFunction|StopFallAudioTimeline)
          (Utilities|Time|ClearandInvalidateTimerbyHandle (Variables|Default|GetFallingAudioCheckTimerHandle)))))))

(fn PlayLandingFX_InWater
  (Networking|CanExecuteCosmeticEvents
    (:True
      ;[1 1 1] Sound, FX, and Postprocess
      (Effects|Components|ParticleSystem|SpawnFXAtLocation "/Game/Gadgets/Assets/SupplyDrop/Effects/NS_Supply_Drop_Impact_Water.NS_Supply_Drop_Impact_Water" (Math|Vector|MakeVector (.x (.location (Transformation|GetWorldTransform (Variables|Default|GetCrateCollision)))) (.y (.location (Transformation|GetWorldTransform (Variables|Default|GetCrateCollision)))) (- (.z (.location (Transformation|GetWorldTransform (Variables|Default|GetCrateCollision)))) (.z (Variables|/Script/Engine.BoxComponent|GetBoxExtent (Variables|Default|GetCrateCollision)))))))))

(fn SpawnLoot
  (CallFunction|SpawnVehicle)
  (Networking|SwitchHasAuthority
    (:Authority
      (Utilities|FlowControl|DoOnce false
        (:Completed
          (bind _returnvalue (Loot|GetLootTierGroupOverride))
          (Variables|Default|SetLootTableName (select (Utilities|Name|Equal(Name) _returnvalue "None") (Variables|Default|GetLootTableName) _returnvalue))
          (if (Utilities|Name|Equal(Name) (Variables|FortAthenaSupplyDrop|GetLootTableNameOverride) "None")
            (Variables|Default|SetLootTableNameToUse (Variables|Default|GetLootTableName))
            (else
              (Variables|Default|SetLootTableNameToUse (Variables|FortAthenaSupplyDrop|GetLootTableNameOverride))))
          (bind (_outloottodrop _returnvalue_2) (Loot|PickLootDrops (Variables|Default|GetLootTableNameToUse) (Loot|GetLootLevel) -1 "(GameplayTags=)"))
          (if _returnvalue_2
            (CallFunction|Analytics _outloottodrop)
            (if true
              (Variables|Default|SetDelayedSpawnLoot _outloottodrop)
              (CallFunction|OnDelayedSpawnLoot)
              (else
                (for _array_element _outloottodrop
                  (bind _asfort_world_item_definition (Utilities|Casting|CastToFortWorldItemDefinition (Class|FortItemFunctionLibrary|BreakFortItemEntry _array_element)))
                  (bind _returnvalue_3 (CallFunction|SpawnPickup _asfort_world_item_definition (.Count (Class|FortItemFunctionLibrary|BreakFortItemEntry _array_element)) (Variables|Default|GetMyInteractingPawn) (CallFunction|GetLootSpawnLocation)))
                  ;[1 1 1] apply mod set functions
                  (bind _asfort_weapon_item_definition (Utilities|Casting|CastToFortWeaponItemDefinition _asfort_world_item_definition)
                    (:then
                      (bind _returnvalue_4 (Class|FortWeaponModFunctionLibrary|GetRandomWeaponModSetData _asfort_weapon_item_definition))
                      (Class|FortWeaponModFunctionLibrary|ApplyModSettoPickup _returnvalue_4 _returnvalue_3))
                    (:CastFailed
                      _))
                  (CallFunction|OnLootSpawned))
                (Actor|DestroyActor)))))))))

(fn SpawnVehicle
  (Utilities|FlowControl|DoOnce false
    (:Completed
      (Game|SpawnActorfromClass (Variables|Default|GetVehicleType) (Math|Transform|MakeTransform (Transformation|GetActorLocation)) "AdjustIfPossibleButAlwaysSpawn" "SelectDefaultAtRuntime" 0 false 0))))

(fn UpdateReticleLocation
  (bind _tracedistance (LocalVariables|GetTraceDistance))
  (bind _returnvalue (+ _tracedistance 1e+03))
  (bind _falldirection (Variables|Default|GetFallDirection))
  (bind _returnvalue_2 (* _falldirection _returnvalue))
  (bind _cratecollision (Variables|Default|GetCrateCollision))
  (bind _returnvalue_3 (Transformation|GetWorldLocation _cratecollision))
  (bind _returnvalue_4 (+ _returnvalue_3 _returnvalue_2))
  (bind _outhit (GroundCollision|GroundTrace _returnvalue_3 _returnvalue_4))
  (bind (_bblockinghit _binitialoverlap _time _distance _location) (Collision|BreakHitResult _outhit))
  (bind (_returnvalue_x _returnvalue_y _returnvalue_z) (Math|Vector|BreakVector (- (Transformation|GetWorldLocation (Variables|Default|GetCrateCollision)) (Variables|Default|GetGroundLocation))))
  (bind _watercollisionsphere (Variables|Default|GetWaterCollisionSphere))
  (bind _reticle_niagara (Variables|Default|GetReticle_Niagara))
  (if (not (Variables|Default|GetbGroundLocationInitialized))
    ;[1 1 1] Initializing the GroundLocation for the first time, so attempting to double the FallHeight for the initial trace distance
    (LocalVariables|SetTraceDistance (* (Variables|Default|GetFallHeight) 2.0))
    (else
      ;[1 1 1] Updating the GroundLocation and using the previous location to determine the trace distance
      (bind _distance_2 (Class|MissionBlueprintFunctionLibrary|DistanceBetweenTwoVectors (Transformation|GetWorldLocation (Variables|Default|GetCrateCollision)) (Variables|Default|GetGroundLocation)))
      (LocalVariables|SetTraceDistance _distance_2)))
  (Variables|Default|SetbHasValidGroundLocation _bblockinghit)
  (bind _output_get (Variables|Default|SetGroundLocation (Math|Vector|SelectVector _location _returnvalue_4 (or _bblockinghit (Variables|Default|GetbUseDefaultLocationOnGroundTraceFail)))))
  ;[1 1 1] Only proceed if ground location changes (or updating first time)
  (if (or (not (Variables|Default|GetbGroundLocationInitialized)) (Utilities|Operators|NotEqual(!=) (Variables|Default|GetGroundLocation) (Variables|Default|GetLastGroundLocation) 0.01))
    (Utilities|FlowControl|Sequence
      (:then_0
        ;[1 1 1] Move reticle FX to the ground
        (Networking|CanExecuteCosmeticEvents
          (:True
            (if (Utilities|IsValid _reticle_niagara :!class /Script/Engine.KismetSystemLibrary)
              (Transformation|SetWorldLocation _reticle_niagara _output_get false false)))))
      (:then_1
        ;[1 1 1] Toggle overlaps on for the WaterCollisionSphere only if we could possibly hit water before the ground location. If no water, detach component and disable collision and overlaps, so we pay no cost for having it.
        (bind (_outhit_2 _returnvalue_5) (Collision|SphereTraceByChannel (Transformation|GetWorldLocation (Variables|Default|GetCrateCollision)) (Variables|Default|GetGroundLocation) (Components|Sphere|GetScaledSphereRadius (Variables|Default|GetWaterCollisionSphere)) "TraceTypeQuery12"))
        (if _returnvalue_5
          (Transformation|AttachComponentToComponent _watercollisionsphere (Variables|Default|GetCrateCollision) "None" "KeepRelative" "KeepRelative" "KeepRelative" true)
          (Collision|SetCollisionEnabled _watercollisionsphere "QueryOnly")
          (Variables|/Script/Engine.PrimitiveComponent|SetbGenerateOverlapEvents true _watercollisionsphere)
          (else
            (Variables|/Script/Engine.PrimitiveComponent|SetbGenerateOverlapEvents false _watercollisionsphere)
            (Collision|SetCollisionEnabled _watercollisionsphere "NoCollision")
            (Transformation|DetachFromComponent _watercollisionsphere "KeepRelative" "KeepRelative" "KeepRelative" true))))
      (:then_2
        (Variables|Default|SetbGroundLocationInitialized true))
      (:then_3
        (Variables|Default|SetLastGroundLocation (Variables|Default|GetGroundLocation))))
    (else
      ;[1 1 1] On client: when crate is close to the ground, turn back on projectile collision sweeps. This helps the case when the balloon has not yet popped and it's nearing the ground (popping already turns on sweeps because of the high fall speed)
      (Networking|SwitchHasAuthority
        (:Remote
          (if (<= _returnvalue_z (Variables|Default|GetClientHeightToEnableCollisionSweeps))
            (Variables|/Script/Engine.ProjectileMovementComponent|SetbSweepCollision true (Variables|Default|GetFortProjectileMovement))))))))

(fn ConstructionScript
  (Components|Movement|SetUpdatedComponent (Variables|Default|GetFortProjectileMovement) (Variables|Default|GetCrateCollision))
  (Transformation|DetachFromComponent (Variables|Default|GetReticle_Niagara) "KeepWorld" "KeepWorld" "KeepRelative" true))
