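; Blueprint variant of the Athena supply drop, named for the Vault, that handles drops spawning on the ground instead of in the sky.
; Two seconds after BeginPlay, if the ground location is initialized, it calls PlayLandingFX with false, intended to disable the "landed" sound.
; BlueprintCanInteract always returns true, the construction script only calls its parent, and the overlap and tick events are empty.
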(event EventBeginPlay
  ;[0 0.06024 0.1562 1 17] Disable supply drop "landed" sound in situations where the supply drop spawns on the ground vs. up in the sky.
  (Utilities|FlowControl|Delay 2.0)
  (if (Variables|Default|GetbGroundLocationInitialized)
    (CallFunction|PlayLandingFX false)))

(event Collision|EventActorBeginOverlap
  (params (OtherActor :type Actor%20Object%20Reference)))

(event EventTick
  (params (DeltaSeconds :type Float%20%28single-precision%29)))

(fn BlueprintCanInteract
  (params (InteractingPawn :type Fort%20Pawn%20Object%20Reference) (InteractionBeingAttempted :type EInteractionBeingAttempted%20Enum :default "FirstInteraction") (InteractionType :type TInteractionType%20Enum :default "IT_NoInteraction"))
  (return true))

(fn ConstructionScript
  (|CallParentFunction))
