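; Damageable balloon for Athena supply drops that notifies the drop when shot down or when the crate is gone.
; ShouldDie always returns false; server damage at Health ≤ 1 instead broadcasts OnBalloonDestroyed, sets replicated DestroyBalloon and fires shot-down analytics.
; OnRep_DestroyBalloon spawns explosion FX at the "boom" socket, hides the health bar and destroys the actor.
; The health bar stays hidden until damaged; crate-triggered removal flags destroy audio off.
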(event Custom|CrateIsGone
  ;[1 1 1] crate is gone (looted/destroyed)
  (Variables|Default|SetDontPlayDestroyAudio true)
  (Variables|Actor|SetbCanBeDamaged false)
  ;[1 1 1] let the supply drop know
  (Default|CallOnBalloonDestroyed)
  ;[1 1 1] destroy balloon
  (Variables|Default|SetDestroyBalloon true))

(event EventOnDamageServer
  (params (Damage :type Float%20%28single-precision%29) (DamageTags :type Gameplay%20Tag%20Container%20Structure%20%28by%20ref%29) (Momentum :type Vector :default "0, 0, 0") (HitInfo :type Hit%20Result%20Structure%20%28by%20ref%29) (InstigatedBy :type Controller%20Object%20Reference) (DamageCauser :type Actor%20Object%20Reference) (EffectContext :type Gameplay%20Effect%20Context%20Handle%20Structure))
  ;[1 1 1] handle damage
  (if (Variables|Actor|GetbCanBeDamaged)
    (Utilities|FlowControl|Sequence
      (:then_0
        (if (<= (Health|GetHealth) 1)
          (Utilities|FlowControl|Sequence
            (:then_0
              (Variables|Actor|SetbCanBeDamaged false)
              ;[1 1 1] let the supply drop know
              (Default|CallOnBalloonDestroyed)
              ;[1 1 1] destroy balloon
              (Variables|Default|SetDestroyBalloon true))
            (:then_1
              ;[1 1 1] Fire analytics event indicating that the balloon was shot down
              (Networking|SwitchHasAuthority
                (:Authority
                  (Analytics|FireAnalyticsEventBlueprint "AthenaSupplyDrop.ShotDown" (Utilities|Struct|MakeFortAnalyticsEventAttribute "(Name=\"\",Value=\"\")"))))))))
      (:then_1
        ;[1 1 1] make hp bar visible after taking damage
        (Utilities|FlowControl|DoOnce false
          (:Completed
            (HealthBarComponent|SetHealthBarVisibility (Variables|Default|GetFortHealthBar) true)))))))

(event EventBeginPlay
  ;[1 1 1] hide health bar on beginplay
  (HealthBarComponent|SetHealthBarVisibility (Variables|Default|GetFortHealthBar) false))

(fn OnRep_DestroyBalloon
  (bind _returnvalue (Effects|Components|ParticleSystem|SpawnFXAtLocation "/Game/Gadgets/Assets/BottleRocket/NS_BottleRocket_Explo.NS_BottleRocket_Explo" (Transformation|GetSocketLocation (Variables|Debug|GetStaticMeshComponent) "boom")))
  (Components|Activation|SetActive _returnvalue true false)
  (HealthBarComponent|SetHealthBarVisibility (Variables|Default|GetFortHealthBar) false)
  (Actor|DestroyActor))

(fn ShouldDie
  (params (Damage :type Float%20%28single-precision%29) (EventInstigator :type Controller%20Object%20Reference) (DamageCauser :type Actor%20Object%20Reference) (bIsTest :type Boolean) (EffectContext :type Gameplay%20Effect%20Context%20Handle%20Structure%20%28by%20ref%29))
  (return false))

(fn ConstructionScript)
