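; Blueprint for the Athena gift supply drop, apparently a child of a base supply drop class.
; Its HideAndExplodeCrate event calls the parent behavior, then sets the gift mesh to hidden in game.
; The construction script only calls the parent version.
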(event EventHideAndExplodeCrate
  (|CallParentFunction)
  (Development|SetHiddeninGame (Variables|Default|GetGiftMesh) true false))

(fn ConstructionScript
  (|CallParentFunction))
